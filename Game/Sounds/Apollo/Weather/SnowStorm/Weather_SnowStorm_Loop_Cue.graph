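# Looping snowstorm weather ambience cue that fades in at start, with overall volume driven by Blend.
# An alpine bed, picked by EventIndex time of day and quality level, is scaled by Intensity and mixed with wind.
# Intensity crossfades light and heavy wind layers, combining slowly undulating loops with randomly delayed buffeting one-shots.
# High Blend values crossfade the shared wind branch from its attenuation asset to a non-spatialized, unattenuated override.
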SoundCue /Game/Sounds/Apollo/Weather/SnowStorm/Weather_SnowStorm_Loop_Cue
root=Enveloper_2
nodes [
 {id=Attenuation_0, children [Switch_0], properties {attenuationSettings=&Fort_Ambient_Athena_Att}}
 {
  id=Attenuation_1
  children [ParamCrossFade_0]
  properties {
   attenuationSettings=/Game/Sounds/Attenuation/Fort_Ambient_Athena_SnowStorm_Wind_Att
   attenuationOverrides {bAttenuate=false, bSpatialize=false}
  }
 }
 {
  id=Attenuation_2
  children [ParamCrossFade_0]
  properties {
   attenuationSettings=&Fort_Ambient_Athena_SnowStorm_Wind_Att
   bOverrideAttenuation=true
   attenuationOverrides {bAttenuate=false, bSpatialize=false}
  }
 }
 {id=Delay_0, children [Random_0], properties {delayMax=50, delayMin=10}}
 {id=Delay_1, children [Random_1], properties {delayMax=20, delayMin=5}}
 {
  id=Enveloper_0
  children [ParamCrossFade_2]
  properties {
   bLoop=true
   bLoopIndefinitely=true
   loopEnd=60
   pitchCurve {editorCurveData {keys [{time=0}]}}
   volumeCurve {editorCurveData {keys [<at 0 1>,<at 30 0.75>,<at 60 1>]}}
  }
 }
 {
  id=Enveloper_1
  children [ParamCrossFade_1]
  properties {
   bLoop=true
   bLoopIndefinitely=true
   loopEnd=60
   pitchCurve {editorCurveData {keys [{time=0}]}}
   volumeCurve {editorCurveData {keys [<at 0 0.75>,<at 30 1>,<at 60 0.75>]}}
  }
 }
 {
  id=Enveloper_2
  children [ModulatorContinuous_0]
  properties {pitchCurve {editorCurveData {keys [{time=0}]}}, volumeCurve {editorCurveData {keys [<at 0 0>,<at 0.5 0>,<at 2 1>]}}}
 }
 {id=Looping_0, children [Delay_0]}
 {id=Looping_1, children [Delay_1]}
 {id=Mixer_0, children [ModulatorContinuous_1, ParamCrossFade_3], properties {inputVolume [1,1]}}
 {id=Mixer_2, children [Enveloper_1,Looping_0], properties {inputVolume [0.5,1]}}
 {id=Mixer_3, children [Enveloper_0, ParamCrossFade_4], properties {inputVolume [1,1]}}
 {id=ModulatorContinuous_0, children [Mixer_0], properties {volumeModulationParams {default=0, parameterName=Blend}}}
 {
  id=ModulatorContinuous_1
  children [Attenuation_0]
  properties {volumeModulationParams {default=0, maxOutput=2, minOutput=0.4, parameterName=Intensity}}
 }
 {
  id=ParamCrossFade_0
  children [Mixer_2,Mixer_3]
  properties {
   paramName=Intensity
   crossFadeInput [
    {fadeInDistanceEnd=0.25, fadeInDistanceStart=0, fadeOutDistanceEnd=100, fadeOutDistanceStart=100, volume=1}
    {fadeInDistanceEnd=1, fadeInDistanceStart=0, fadeOutDistanceEnd=100, fadeOutDistanceStart=100, volume=1}
   ]
  }
 }
 {
  id=ParamCrossFade_1
  children [WavePlayer_15]
  properties {paramName=Intensity, crossFadeInput [{fadeInDistanceEnd=0, fadeInDistanceStart=0, fadeOutDistanceEnd=1, fadeOutDistanceStart=0, volume=2}]}
 }
 {
  id=ParamCrossFade_2
  children [WavePlayer_17, WavePlayer_6]
  properties {
   paramName=Intensity
   crossFadeInput [
    {fadeInDistanceEnd=1, fadeInDistanceStart=0.5, fadeOutDistanceEnd=100, fadeOutDistanceStart=100, volume=2}
    {fadeInDistanceEnd=1, fadeInDistanceStart=0.9, fadeOutDistanceEnd=100, fadeOutDistanceStart=100, volume=2}
   ]
  }
 }
 {
  id=ParamCrossFade_3
  children [Attenuation_1, Attenuation_2]
  properties {
   paramName=Blend
   crossFadeInput [
    {fadeInDistanceEnd=0, fadeInDistanceStart=0, fadeOutDistanceEnd=1, fadeOutDistanceStart=0.75, volume=1}
    {fadeInDistanceEnd=1, fadeInDistanceStart=0.75, fadeOutDistanceEnd=100, fadeOutDistanceStart=100, volume=1}
   ]
  }
 }
 {
  id=ParamCrossFade_4
  children [Looping_1]
  properties {
   paramName=Intensity
   crossFadeInput [{fadeInDistanceEnd=1, fadeInDistanceStart=0.5, fadeOutDistanceEnd=100, fadeOutDistanceStart=100, volume=2}]
  }
 }
 {id=QualityLevel_0, children [WavePlayer_2,WavePlayer_2,WavePlayer_3,WavePlayer_3]}
 {id=QualityLevel_1, children [WavePlayer_0,WavePlayer_0,WavePlayer_1,WavePlayer_1]}
 {id=QualityLevel_2, children [WavePlayer_4,WavePlayer_4,WavePlayer_5,WavePlayer_5]}
 {id=Random_0, children [WavePlayer_11, WavePlayer_12, WavePlayer_13], properties {weights [1,1,1]}}
 {id=Random_1, children [WavePlayer_7,WavePlayer_8,WavePlayer_9], properties {weights [1,1,1]}}
 {
  id=Switch_0
  children [QualityLevel_2, QualityLevel_2, QualityLevel_2, QualityLevel_1, QualityLevel_0]
  properties {intParameterName=EventIndex}
 }
 {id=WavePlayer_0, properties {bLooping=true, soundWaveAssetPtr=/Game/Sounds/Apollo/Biomes/Alpine/Biome_Alpine_Loop_Evening_Quad}}
 {id=WavePlayer_1, properties {bLooping=true, soundWaveAssetPtr=/Game/Sounds/Apollo/Biomes/Alpine/Biome_Alpine_Loop_Evening_ST}}
 {id=WavePlayer_11, properties {soundWaveAssetPtr=&/Weather_SnowStorm_Buffeting_Light_OneShot_01}}
 {id=WavePlayer_12, properties {soundWaveAssetPtr=&/Weather_SnowStorm_Buffeting_Light_OneShot_02}}
 {id=WavePlayer_13, properties {soundWaveAssetPtr=&/Weather_SnowStorm_Buffeting_Light_OneShot_03}}
 {
  id=WavePlayer_15
  properties {bLooping=true, soundWaveAssetPtr=&/Weather_SnowStorm_Accumulating_Light_Loop_01}
 }
 {
  id=WavePlayer_17
  properties {bLooping=true, soundWaveAssetPtr=&/Weather_SnowStorm_Accumulating_Heavy_Loop_01}
 }
 {id=WavePlayer_2, properties {bLooping=true, soundWaveAssetPtr=/Game/Sounds/Apollo/Biomes/Alpine/Biome_Alpine_Loop_Night_Quad}}
 {id=WavePlayer_3, properties {bLooping=true, soundWaveAssetPtr=/Game/Sounds/Apollo/Biomes/Alpine/Biome_Alpine_Loop_Night_ST}}
 {id=WavePlayer_4, properties {bLooping=true, soundWaveAssetPtr=/Game/Sounds/Apollo/Biomes/Alpine/Biome_Alpine_Loop_Morning_Quad}}
 {id=WavePlayer_5, properties {bLooping=true, soundWaveAssetPtr=/Game/Sounds/Apollo/Biomes/Alpine/Biome_Alpine_Loop_Morning_ST}}
 {id=WavePlayer_6, properties {bLooping=true, soundWaveAssetPtr=&/Weather_SnowStorm_Buffeting_Heavy_Loop}}
 {id=WavePlayer_7, properties {soundWaveAssetPtr=&/Weather_SnowStorm_Buffeting_Heavy_OneShot_01}}
 {id=WavePlayer_8, properties {soundWaveAssetPtr=&/Weather_SnowStorm_Buffeting_Heavy_OneShot_02}}
 {id=WavePlayer_9, properties {soundWaveAssetPtr=&/Weather_SnowStorm_Buffeting_Heavy_OneShot_03}}
]
properties {maxDistance=2097152, soundClassObject=/Game/Sounds/FortSoundClassesAndModes/Fort_Ambient_DuckOutdoors_Athena_PSM, volumeMultiplier=0.501}
