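# Sound cue for the start of a snowstorm, using the snowstorm wind attenuation and an outdoor-ducking ambient sound class.
# A mixer layers buffeting one-shots with quieter storm "tell" sounds that start after a fixed six-second delay.
# In each layer, an Intensity parameter crossfades between light and heavy sets, each picking a random wave.
# Light sets fade out at high intensity, while heavy sets rise across the range.
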SoundCue /Game/Sounds/Apollo/Weather/SnowStorm/Weather_SnowStorm_Start_Cue
root=Mixer_0
nodes [
 {id=Delay_0, children [ParamCrossFade_0], properties {delayMax=6, delayMin=6}}
 {id=Mixer_0, children [ParamCrossFade_1, Delay_0], properties {inputVolume [1,0.6]}}
 {
  id=ParamCrossFade_0
  children [Random_3,Random_0]
  properties {
   paramName=Intensity
   crossFadeInput [
    {fadeInDistanceEnd=0.5, fadeInDistanceStart=0.25, fadeOutDistanceEnd=1, fadeOutDistanceStart=0.5, volume=0.5}
    {fadeInDistanceEnd=1, fadeInDistanceStart=0, fadeOutDistanceEnd=100, fadeOutDistanceStart=100, volume=1}
   ]
  }
 }
 {
  id=ParamCrossFade_1
  children [Random_1,Random_2]
  properties {
   paramName=Intensity
   crossFadeInput [
    {fadeInDistanceEnd=0.5, fadeInDistanceStart=0.25, fadeOutDistanceEnd=1, fadeOutDistanceStart=0.5, volume=1}
    {fadeInDistanceEnd=1, fadeInDistanceStart=0, fadeOutDistanceEnd=100, fadeOutDistanceStart=100, volume=1}
   ]
  }
 }
 {id=Random_0, children [WavePlayer_0,WavePlayer_1,WavePlayer_2,WavePlayer_3,WavePlayer_4], properties {weights [1,1,1,1,1]}}
 {id=Random_1, children [WavePlayer_11, WavePlayer_12, WavePlayer_13], properties {weights [1,1,1]}}
 {id=Random_2, children [WavePlayer_7,WavePlayer_8,WavePlayer_9], properties {weights [1,1,1]}}
 {id=Random_3, children [WavePlayer_5, WavePlayer_6, WavePlayer_10, WavePlayer_14, WavePlayer_15], properties {weights [1,1,1,1,1]}}
 {id=WavePlayer_0, properties {soundWaveAssetPtr=&/Weather_SnowStorm_Tell_Heavy_01}}
 {id=WavePlayer_1, properties {soundWaveAssetPtr=&/Weather_SnowStorm_Tell_Heavy_02}}
 {id=WavePlayer_10, properties {soundWaveAssetPtr=&/Weather_SnowStorm_Tell_Light_02}}
 {id=WavePlayer_11, properties {soundWaveAssetPtr=&/Weather_SnowStorm_Buffeting_Light_OneShot_01}}
 {id=WavePlayer_12, properties {soundWaveAssetPtr=&/Weather_SnowStorm_Buffeting_Light_OneShot_02}}
 {id=WavePlayer_13, properties {soundWaveAssetPtr=&/Weather_SnowStorm_Buffeting_Light_OneShot_03}}
 {id=WavePlayer_14, properties {soundWaveAssetPtr=&/Weather_SnowStorm_Tell_Light_03}}
 {id=WavePlayer_15, properties {soundWaveAssetPtr=&/Weather_SnowStorm_Tell_Light_04}}
 {id=WavePlayer_2, properties {soundWaveAssetPtr=&/Weather_SnowStorm_Tell_Heavy_03}}
 {id=WavePlayer_3, properties {soundWaveAssetPtr=&/Weather_SnowStorm_Tell_Heavy_04}}
 {id=WavePlayer_4, properties {soundWaveAssetPtr=&/Weather_SnowStorm_Tell_Heavy_05}}
 {id=WavePlayer_5, properties {soundWaveAssetPtr=&/Weather_SnowStorm_Tell_Light_05}}
 {id=WavePlayer_6, properties {soundWaveAssetPtr=&/Weather_SnowStorm_Tell_Light_01}}
 {id=WavePlayer_7, properties {soundWaveAssetPtr=&/Weather_SnowStorm_Buffeting_Heavy_OneShot_01}}
 {id=WavePlayer_8, properties {soundWaveAssetPtr=&/Weather_SnowStorm_Buffeting_Heavy_OneShot_02}}
 {id=WavePlayer_9, properties {soundWaveAssetPtr=&/Weather_SnowStorm_Buffeting_Heavy_OneShot_03}}
]
properties {
 attenuationSettings=/Game/Sounds/Attenuation/Fort_Ambient_Athena_SnowStorm_Wind_Att
 maxDistance=2097152
 soundClassObject=/Game/Sounds/FortSoundClassesAndModes/Fort_Ambient_DuckOutdoors_Athena_PSM
 volumeMultiplier=0.501
}
