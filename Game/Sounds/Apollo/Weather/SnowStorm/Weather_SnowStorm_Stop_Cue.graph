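# Snowstorm stop sound cue that plays a one-shot wind buffeting sound as outdoor ambient weather audio.
# After a random two to three second delay, a crossfade driven by the Intensity parameter blends from light buffeting at low values to heavy buffeting at high values.
# Each side picks randomly among three equally weighted wave variations.
# Playback uses snowstorm wind attenuation and the outdoor ambient sound class at reduced volume.
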SoundCue /Game/Sounds/Apollo/Weather/SnowStorm/Weather_SnowStorm_Stop_Cue
root=Delay_0
nodes [
 {id=Delay_0, children [ParamCrossFade_1], properties {delayMax=3, delayMin=2}}
 {
  id=ParamCrossFade_1
  children [Random_1,Random_2]
  properties {
   paramName=Intensity
   crossFadeInput [
    {fadeInDistanceEnd=0, fadeInDistanceStart=0, fadeOutDistanceEnd=1, fadeOutDistanceStart=0, volume=1}
    {fadeInDistanceEnd=1, fadeInDistanceStart=0, fadeOutDistanceEnd=100, fadeOutDistanceStart=100, volume=1}
   ]
  }
 }
 {id=Random_1, children [WavePlayer_11, WavePlayer_12, WavePlayer_13], properties {weights [1,1,1]}}
 {id=Random_2, children [WavePlayer_7,WavePlayer_8,WavePlayer_9], properties {weights [1,1,1]}}
 {id=WavePlayer_11, properties {soundWaveAssetPtr=&/Weather_SnowStorm_Buffeting_Light_OneShot_01}}
 {id=WavePlayer_12, properties {soundWaveAssetPtr=&/Weather_SnowStorm_Buffeting_Light_OneShot_02}}
 {id=WavePlayer_13, properties {soundWaveAssetPtr=&/Weather_SnowStorm_Buffeting_Light_OneShot_03}}
 {id=WavePlayer_7, properties {soundWaveAssetPtr=&/Weather_SnowStorm_Buffeting_Heavy_OneShot_01}}
 {id=WavePlayer_8, properties {soundWaveAssetPtr=&/Weather_SnowStorm_Buffeting_Heavy_OneShot_02}}
 {id=WavePlayer_9, properties {soundWaveAssetPtr=&/Weather_SnowStorm_Buffeting_Heavy_OneShot_03}}
]
properties {
 attenuationSettings=/Game/Sounds/Attenuation/Fort_Ambient_Athena_SnowStorm_Wind_Att
 maxDistance=2097152
 soundClassObject=&Fort_Ambient_Outdoors_Athena
 volumeMultiplier=0.501
 preEffectBusSends [{}]
}
